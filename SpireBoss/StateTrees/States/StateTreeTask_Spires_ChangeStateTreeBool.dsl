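; StateTree task for the Spires boss that writes a boolean value into a bound StateTree property.
; On EnterState it resolves the bool binding as a property reference and sets it to the configured new value.
; It finishes immediately, succeeding when the reference resolves and failing otherwise.
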(event EventEnterState
  (params (Transition :type State%20Tree%20Transition%20Result%20Structure%20%28by%20ref%29))
  (bind _reference (StateTree|PropertyRef|GetPropertyReference (Variables|Default|GetBoolBinding))
    (:Succeeded
      (Variables|SetBy-RefVar _reference (Variables|Default|GetNew%20Value))
      (StateTree|FinishTask true))
    (:Failed
      (StateTree|FinishTask false))))
