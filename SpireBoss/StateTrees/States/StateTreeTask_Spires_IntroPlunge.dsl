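; State Tree task that starts the intro plunge of a Spire boss actor toward a target anchor.
; On entering the state, it casts the actor to move to BGA_MajorSpireDummy and passes it the anchor's location.
; That actor performs the plunge on server and clients, since the State Tree lacks access to that logic.
; The task finishes successfully when the actor's OnFlowEnded event fires.
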(event EventEnterState
  (params (Transition :type State%20Tree%20Transition%20Result%20Structure%20%28by%20ref%29))
  ;[0.02083 0.02083 0.02083 0.5] We use the actor to perform the plunge on client and server as the state tree doesn't have access to that logic
  (bind _asbga_major_spire_dummy (Utilities|Casting|CastToBGA_MajorSpireDummy (Variables|Default|GetActorToMove)))
  (Class|BGAMajorSpireDummy|PlungeonServerandClients _asbga_major_spire_dummy (Transformation|GetActorLocation (Variables|Default|GetTargetLocationAnchor)))
  (Default|BindEventtoOnFlowEnded _asbga_major_spire_dummy (delegate Custom|OnFlowEnded)))

(event Custom|OnFlowEnded
  ;[0.02083 0.02083 0.02083 0.5] We use the actor to perform the plunge on client and server as the state tree doesn't have access to that logic
  (StateTree|FinishTask true))
