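; StateTree task for the Spires boss intro, which makes the main large spire plunge into view.
; On entering the state it unhides the main spire actor from the context, casts it to BGA_Large_Spire and requests a plunge.
; It binds to the spire's plunge-finished delegate and finishes the task successfully when that fires.
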(event EventEnterState
  (params (Transition :type State%20Tree%20Transition%20Result%20Structure%20%28by%20ref%29))
  (bind _mainspireactor (Variables|Context|GetMainSpireActor))
  (Rendering|SetActorHiddenInGame _mainspireactor false)
  (bind _asbga_large_spire (Utilities|Casting|CastToBGA_Large_Spire _mainspireactor))
  (Class|BGALargeSpire|RequestPlunge _asbga_large_spire)
  (Default|AssignOnPlungeFinished _asbga_large_spire (delegate Custom|OnPlungeFinished)))

(event Custom|OnPlungeFinished
  (StateTree|FinishTask true))
